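# Niagara spawn-trail effect for the blue Dazzle glider variant, a one-emitter system of short-lived ribbons.
# Particles split into three partitions, each sampling its own bone or socket on the user skeletal mesh to set position and ribbon facing, with per-partition widths.
# Particles inherit source movement, with curve-driven vertical scale in gameplay.
# Alpha fades over particle life and loop age, while bIsFrontEnd switches spawn rate, lifetime, width and fade curves.
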NiagaraSystem /BRCosmetics/Gadgets/Assets/VinderTech_GliderChute/Glider_Dazzle/FX/Variant/NS_Glider_Dazzle_Blue_SpawnTrails
dependencies {
 usedDataInterfaces [&NiagaraDataInterfaceArrayFloat, &NiagaraDataInterfaceCurve]
 usedDynamicInputs []
 usedModules [&SystemState]
}
emitters {Spawn_Disks_Ribbons {
  emitterUpdateScript [
   &EmitterState(Life Cycle Mode=::NewEnumerator1, Loop Duration=1, UseLoopDelay=false)
   &SpawnRate(
    SpawnRate=&MakeCustomFloatFromBool(Bool=$User.bIsFrontEnd, True Float=66, False Float=120)
    Use Spawn Probability=false
   )
  ]
  particleSpawnScript [
   &InitializeParticle(
    Color=<rgba 2 2 2 1>
    Lifetime=&MakeCustomFloatFromBool(Bool=$User.bIsFrontEnd, True Float=0.3, False Float=0.2)
    UsePositionOffset=false
    Write Material Random=true
   )
   &PartitionParticles(Number Of Partitions=3)
   &SkeletalMeshLocation(
    Bone / Socket Index=$Particles.PartitionParticles.Partition
    Bone Sampling Mode=::NewEnumerator4
    MaskSpawnGroup=false
    Orientation Sampling=::NewEnumerator1
    Sampled Position Offset=<0 0 -5>
    Sampled Position Offset Coordinate Space:ENiagaraSkelMeshTransforms=::NewEnumerator2
    Skeletal Mesh=$User.SkeletalMesh
   )
   /SetVariables(Particles.RibbonWidth=&Multiply_Float(
     A=&SelectFloatFromArray(Direct Array Index=$Particles.PartitionParticles.Partition, Float Selection Array=DataInterface() {propertyValues {FloatData [50,70,60]}})
     B=&MakeCustomFloatFromBool(Bool=$User.bIsFrontEnd, True Float=0.8, False Float=1)
    )) {isSetParameters=true}
   /SetVariables(Particles.RibbonLinkOrder:NiagaraFloat=&OneMinusFloat(Float=&MakeFloatFromInt(Int=$Particles.UniqueID))) {isSetParameters=true}
   /SetVariables(Particles.RibbonFacing=&RotateVectorByQuaternion(Vector To Rotate=<0 1 0>, Quaternion=$Particles.SkeletalMeshLocation.SampledRotation)) {isSetParameters=true}
  ]
  particleUpdateScript [
   &InheritSourceMovement(
    Applied Position Scale=&MakeVector(
     X=1
     Y=1
     Z=&FN_FrontEndFloatModifier(
      Gameplay=&FloatFromCurve(
       FloatCurve=DataInterface() {propertyValues {Curve {keys [
           {arriveTangent=-8.2148, interpMode=RCIM_Cubic, leaveTangent=-8.2148, tangentMode=RCTM_User, time=0.87, value=1}
           {arriveTangent=-1.0611, interpMode=RCIM_Cubic, leaveTangent=-1.0611, tangentMode=RCTM_User, time=0.97, value=0.6}
          ]}}}
       CurveIndex=$System.NormalizedLoopAge
       Scale Curve=1.3
      )
      Front End=1
     )
    )
    ApplyForce=false
    ApplyPosition=true
    ApplyPositionFalloff=false
   )
   &ParticleState()
   &DynamicMaterialParameters(
    Index 0 Param 1=1
    Index 0 Param 2=0.4
    Index 0 Param 4=0.5
    Module.Param0WriteEnabled=false
    Module.Param1WriteEnabled=true
    Module.Param2WriteEnabled=false
    Module.Param3WriteEnabled=true
   )
   &ScaleColor(
    Scale Alpha=&MakeCustomFloatFromBool(
     Bool=$User.bIsFrontEnd
     True Float=&FloatFromCurve(FloatCurve=DataInterface() {propertyValues {Curve {keys [
          {interpMode=RCIM_Cubic, tangentMode=RCTM_User, time=0.1, value=0}
          {interpMode=RCIM_Cubic, tangentMode=RCTM_User, time=0.25, value=1}
          {interpMode=RCIM_Cubic, tangentMode=RCTM_User, time=0.7, value=0}
         ]}}})
     False Float=&FloatFromCurve(FloatCurve=DataInterface() {propertyValues {Curve {keys [
          {interpMode=RCIM_Cubic, tangentMode=RCTM_User, time=0, value=0}
          {interpMode=RCIM_Cubic, tangentMode=RCTM_User, time=0.14, value=1}
          {interpMode=RCIM_Cubic, tangentMode=RCTM_User, time=1, value=0}
         ]}}})
    )
    ScaleA=true
    ScaleRGB=false
   )
   &ScaleColor(
    Scale Alpha=&MakeCustomFloatFromBool(
     Bool=$User.bIsFrontEnd
     True Float=&FloatFromCurve(
      FloatCurve=DataInterface() {propertyValues {Curve {keys [
          {interpMode=RCIM_Cubic, tangentMode=RCTM_User, time=0.15, value=0}
          <at 0.35 0.6>
          {interpMode=RCIM_Cubic, tangentMode=RCTM_User, time=0.85, value=1}
          {tangentMode=RCTM_User, time=0.92, value=0}
         ]}}}
      CurveIndex=$System.NormalizedLoopAge
     )
     False Float=&FloatFromCurve(
      FloatCurve=DataInterface() {propertyValues {Curve {keys [
          {interpMode=RCIM_Cubic, tangentMode=RCTM_User, time=0.15, value=0}
          <at 0.3 1>
          {interpMode=RCIM_Cubic, tangentMode=RCTM_User, time=0.85, value=1}
          {tangentMode=RCTM_User, time=1, value=0}
         ]}}}
      CurveIndex=$System.NormalizedLoopAge
     )
    )
    ScaleA=true
    ScaleRGB=false
   ) {module=ScaleColor001}
   &SolveForcesAndVelocity(Clamp Velocity=false, Limit Acceleration=false)
  ]
  rendererClasses [&NiagaraRibbonRendererProperties]
  renderers [{rendererClass=&NiagaraRibbonRendererProperties}]
 }}
summary {
 systemName=NS_Glider_Dazzle_Blue_SpawnTrails
 emitters [{emitterName=Spawn_Disks_Ribbons, rendererClasses [&NiagaraRibbonRendererProperties]}]
 userVariables [
  User.bIsFrontEnd {type=NiagaraBool}
  User.bIsFrontEndPreview {type=NiagaraBool}
  User.bIsFullyDeployed {type=NiagaraBool}
  User.ForwardDot {type=NiagaraFloat, defaultValue=1, description=""}
  User.PlayerSpeed {type=NiagaraFloat, defaultValue=0, description=""}
  User.RibbonIDOffset {type=NiagaraInt32}
  User.RightDot {type=NiagaraFloat, defaultValue=0, description=""}
  User.SkeletalMesh {
   type=NiagaraDataInterfaceSkeletalMesh
   defaultValue {
    struct=NiagaraExt_VariableValue_DataInterface
    value {dataInterface=&self:NiagaraDataInterfaceSkeletalMesh_0}
   }
  }
 ]
}
